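# Physics asset for the EmeraldGlassPink backpack cosmetic, defining bodies and joints for the dangling sphere and ring bones on the pack.
# Small capsules cover the pack, sphere, sphere tip and two rings, with the tip and ring shapes offset from their bones.
# The sphere is constrained to the pack with narrow swing and twist limits.
# Both rings and the tip branch from the sphere, each with a wider limit on one swing axis only.
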PhysicsAsset /BRCosmetics/Accessories/FORT_Backpacks/Backpack_F_MED_EmeraldGlassPink/Meshes/F_MED_EmeraldGlassPink_Pack_Physics
bodies {
 dyn_ring_1 [{shapeType=Capsule, length=0, radius=0.5, shapeName=dyn_ring_2_capsule, center=<16 0 0>, rotation=<rot 90 0 0>}]
 dyn_ring_2 [{shapeType=Capsule, length=0, radius=1, shapeName=dyn_ring_2_capsule, center=<12 0 0>, rotation=<rot 90 0 0>}]
 dyn_sphere [{shapeType=Capsule, length=1, radius=1, shapeName=dyn_sphere_capsule, center=<1 0 0>, rotation=<rot 90 0 0>}]
 dyn_sphere_tip [{shapeType=Capsule, length=0, radius=1.5, shapeName=dyn_ring_2_capsule, center=<8 0 0>, rotation=<rot 90 0 0>}]
 pack [{shapeType=Capsule, length=1, radius=1, shapeName=pack_capsule}]
}
constraints [
 {bone1Name=dyn_sphere, bone2Name=pack, swing1LimitDegrees=20, swing2LimitDegrees=10, twistLimitDegrees=1}
 {bone1Name=dyn_ring_1, bone2Name=dyn_sphere, swing1LimitDegrees=35, swing2LimitDegrees=0, twistLimitDegrees=0}
 {bone1Name=dyn_sphere_tip, bone2Name=dyn_sphere, swing1LimitDegrees=35, swing2LimitDegrees=0, twistLimitDegrees=0}
 {bone1Name=dyn_ring_2, bone2Name=dyn_sphere, swing1LimitDegrees=35, swing2LimitDegrees=0, twistLimitDegrees=0}
]
