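# Animation Blueprint for the EmeraldGlassPink backpack cosmetic, combining rigid-body physics with a procedural float on its sphere and rings.
# The event graph sets gravity via a deprecated character-state function and builds three real-time sine offset vectors with differing X ranges.
# RigidBody simulates in root-bone space, then additive bone translations apply the offsets before a root scale adjustment.
# A rotation-correction chain is disconnected from the output, and its alpha setter is unwired.
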AnimBlueprint /BRCosmetics/Accessories/FORT_Backpacks/Backpack_F_MED_EmeraldGlassPink/Meshes/F_MED_EmeraldGlassPink_Pack_AnimBP
target_skeleton=&/F_MED_EmeraldGlassPink_Pack_Skeleton
graphs [
 AnimGraph {nodes [
   n1 = Get Rigid Body Correction() -> (Rigid Body Correction:double)
   n2 = "Transform (Modify) Bone - Bone: dyn_sphere"(Alpha=@n1, Translation="0,0,0", Rotation="-20,0.000,-0.000", Scale="1,1,1") {
    rotationMode=BMM_Additive
    rotationSpace=BCS_BoneSpace
    alphaBoolBlend {blendOption=Linear}
    alphaScaleBiasClamp {bInterpResult=true}
    boneToModify {boneName=dyn_sphere}
    rotation {pitch=-20}
   }
   n3 = "Transform (Modify) Bone - Bone: dyn_ring_1"(ComponentPose=@n2, Alpha=@n1, Translation="0,0,0", Rotation="-10,0.000,-0.000", Scale="1,1,1") {
    rotationMode=BMM_Additive
    rotationSpace=BCS_BoneSpace
    alphaBoolBlend {blendOption=Linear}
    alphaScaleBiasClamp {bInterpResult=true}
    boneToModify {boneName=dyn_ring_1}
    rotation {pitch=-10}
   }
   n4 = "Transform (Modify) Bone - Bone: dyn_ring_2"(ComponentPose=@n3, Alpha=@n1, Translation="0,0,0", Rotation="-10,0.000,-0.000", Scale="1,1,1") {
    rotationMode=BMM_Additive
    rotationSpace=BCS_BoneSpace
    alphaBoolBlend {blendOption=Linear}
    alphaScaleBiasClamp {bInterpResult=true}
    boneToModify {boneName=dyn_ring_2}
    rotation {pitch=-10}
   }
   "Transform (Modify) Bone - Bone: dyn_sphere_tip"(ComponentPose=@n4, Alpha=@n1, Translation="0,0,0", Rotation="-10,0.000,-0.000", Scale="1,1,1") {
    rotationMode=BMM_Additive
    rotationSpace=BCS_BoneSpace
    alphaBoolBlend {blendOption=Linear}
    alphaScaleBiasClamp {bInterpResult=true}
    boneToModify {boneName=dyn_sphere_tip}
    rotation {pitch=-10}
   }
   n5 = RigidBody!(
    ComponentPose=Local To Component(LocalPose=Input Pose())
    Alpha
    OverrideWorldGravity:Vector=Get GravityOverride()
    ExternalForce
    ComponentLinearAccScale:Vector="0.4,0.3,0.3"
    ComponentLinearVelScale:Vector="0.4,0.3,0.3"
   ) -> (Pose) {
    bClampLinearTranslationLimitToRefPose=true
    bOverrideWorldGravity=true
    overridePhysicsAsset=&/F_MED_EmeraldGlassPink_Pack_Physics
    simulationSpace=BaseBoneSpace
    alphaBoolBlend {blendOption=Linear}
    alphaScaleBiasClamp {bInterpResult=true}
    baseBoneRef {boneName=root}
    componentAppliedLinearAccClamp=<3000 3000 3000>
    componentLinearAccScale=<0.4 0.3 0.3>
    componentLinearVelScale=<0.4 0.3 0.3>
    overrideWorldGravity {z=-980}
    simSpaceSettings {
     maxAngularAcceleration=0
     maxAngularVelocity=0.3
     maxLinearAcceleration=100
     maxLinearVelocity=100
     velocityScaleZ=0
     worldAlpha=0.5
     externalLinearDragV=<0.2 0.2>
    }
   }
   n6 = Get Ring_A_Float() -> (Ring_A_Float:Vector)
   n7 = "Transform (Modify) Bone - Bone: dyn_sphere"(ComponentPose=@n5, Alpha=0.25, Translation=@n6, Rotation="0,0,0", Scale="1,1,1") {translationMode=BMM_Additive, translationSpace=BCS_BoneSpace, alphaBoolBlend {blendOption=Linear}, boneToModify {boneName=dyn_sphere}}
   n8 = "Transform (Modify) Bone - Bone: dyn_ring_1"(ComponentPose=@n7, Alpha=1, Translation=@n6, Rotation="0,0,0", Scale="1,1,1") {translationMode=BMM_Additive, translationSpace=BCS_BoneSpace, alphaBoolBlend {blendOption=Linear}, boneToModify {boneName=dyn_ring_1}}
   n9 = "Transform (Modify) Bone - Bone: dyn_ring_2"(ComponentPose=@n8, Alpha=1, Translation=Get Ring_B_Float() -> (Ring_B_Float:Vector), Rotation="0,0,0", Scale="1,1,1") {translationMode=BMM_Additive, translationSpace=BCS_BoneSpace, alphaBoolBlend {blendOption=Linear}, boneToModify {boneName=dyn_ring_2}}
   n10 = "Transform (Modify) Bone - Bone: dyn_sphere_tip"(ComponentPose=@n9, Alpha=1, Translation=Get Ring_C_Float() -> (Ring_C_Float:Vector), Rotation="0,0,0", Scale="1,1,1") {translationMode=BMM_Additive, translationSpace=BCS_BoneSpace, alphaBoolBlend {blendOption=Linear}, boneToModify {boneName=dyn_sphere_tip}}
   n11 = "Transform (Modify) Bone - Bone: root"(ComponentPose=@n10, Translation="0,0,0", Rotation="0,0,0", Scale="1.1,1.1,1.1") {scaleMode=BMM_Additive, scaleSpace=BCS_BoneSpace, alphaBoolBlend {blendOption=Linear}, boneToModify {boneName=root}, scale=<1.1 1.1 1.1>}
   Output Pose(Result=Component To Local(ComponentPose=@n11)) {layerGroup=None}
  ]}
 EventGraph {nodes [
   Event BlueprintUpdateAnimation()
   > n1 = Sequence()
   > n2 = Gravity Override State Function DEPRECATED(CharacterPartAnimInstance=Self-Reference(), GameTimeInSeconds=GetGameTimeInSeconds(), GravityOverrideParams=Get Gravity Override Params())
   > Set GravityOverride(GravityOverride=@n2)
   n3 = GetRealTimeSeconds()
   n4 = "Sin (Radians)"(A=float * float(A=float * float(A=1.2, B=1.3), B=@n3))
   n5 = float * float(A=1.1, B=1.1)
   n6 = float * float(A="Sin (Radians)"(A=float * float(A=@n5, B=@n3)), B=0.15)
   n7 = float * float(A=0.9, B=1.1)
   n8 = float * float(A="Sin (Radians)"(A=float * float(A=@n7, B=@n3)), B=0.3)
   Set Ring_A_Float(
    execute=@n1.then_1
    Ring_A_Float:Vector=MakeVector(X=Remap Float(Input=@n4, OldRange=(X=-1,Y=1), NewRange=(X=-1,Y=0.2)) -> (Output=~), Y=@n6, Z=@n8)
   ) -> (Output_Get:Vector)
   > Set Ring_B_Float(Ring_B_Float:Vector=MakeVector(X=Remap Float(Input=@n4, OldRange=(X=-1,Y=1), NewRange=(X=-4,Y=0.225)) -> (Output=~), Y=@n6, Z=@n8)) -> (Output_Get:Vector)
   > Set Ring_C_Float(Ring_C_Float:Vector=MakeVector(X=Remap Float(Input=@n4, OldRange=(X=-1,Y=1), NewRange=(X=-7,Y=0.25)) -> (Output=~), Y=@n6, Z=@n8)) -> (Output_Get:Vector)
   n9 = Get bIsSkydiveFloating()
   Set Rigid Body Correction(Rigid Body Correction:double=SelectFloat(A=SelectFloat(A=0.4, B=1, bPickA=@n9), B=0, bPickA=Get bRigidBodyEnabled())) -> (Output_Get:double)
  ]}
]
variables [
 Gravity Override Params {
  type=GravityOverrideParamsStruct
  default {
   emoteOrMelee {gravityOverride=<0 0 -700>}
   onGround_Falling {gravityOverride=<0 0 -700>}
   onGround_Moving {gravityOverride=<0 0 -700>}
   onGround_Standing {gravityOverride=<0 0 -850>}
  }
 }
 GravityOverride {type=Vector}
 Rigid Body Correction {type=number}
 Ring_A_Float {type=Vector}
 Ring_B_Float {type=Vector}
 Ring_C_Float {type=Vector}
]
